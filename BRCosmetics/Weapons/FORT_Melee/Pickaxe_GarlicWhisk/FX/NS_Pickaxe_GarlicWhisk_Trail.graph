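# Niagara swing trail for the GarlicWhisk pickaxe, combining two ribbons with a short spray of petal sprites.
# Trail and TrailTip use shared AnimTrail modules to follow skeletal mesh bones or sockets, fading and narrowing over short lifetimes.
# Time-of-day functions tint both ribbons, while Trail also adjusts color for the front end and animates dynamic material parameters.
# Petals spawn on a decaying rate curve and tumble under wind, curl noise and drag, aligned to mesh orientation.
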NiagaraSystem /BRCosmetics/Weapons/FORT_Melee/Pickaxe_GarlicWhisk/FX/NS_Pickaxe_GarlicWhisk_Trail
dependencies {
 usedDataInterfaces [&NiagaraDataInterfaceCurve, &NiagaraDataInterfaceSkeletalMesh]
 usedDynamicInputs []
 usedModules [
  &self:SystemSpawnScript.SystemScriptSource.SystemScriptGraph.NiagaraNodeAssignment_2.SetVariables_#1
  &SystemState
 ]
}
emitters {
 Petals {
  emitterUpdateScript [
   &EmitterState(Life Cycle Mode=::NewEnumerator1, Loop Behavior=::NewEnumerator1, Loop Duration=0.4, UseLoopDelay=false)
   &SpawnRate(
    SpawnRate=&FloatFromCurve(
     FloatCurve=DataInterface() {propertyValues {Curve {keys [<at 0.1637 1 RCIM_Cubic>,<at 0.6727 0>]}}}
     CurveIndex=$Emitter.NormalizedLoopAge
     Scale Curve=35
    )
    Use Spawn Probability=false
   )
  ]
  particleSpawnScript [
   &InitializeParticle(
    Lifetime Max=0.6
    Lifetime Min=0.3
    Lifetime Mode=::NewEnumerator1
    Mass Max=1.5
    Mass Mode=::NewEnumerator3
    Sprite Rotation Mode=::NewEnumerator2
    Sprite Size Mode=::NewEnumerator1
    Uniform Sprite Size Max=4.5
    Uniform Sprite Size Min=3
    UsePositionOffset=false
    Write Material Random=true
   )
   &ShapeLocation(MaskSpawnGroup=false, Override Seed=false, Sphere Radius=15, UseOwnerScale=true)
   &InitialMeshOrientation(Use Rotation Vector=false)
   &AddVelocity(Velocity=&RandomRangeVector(Minimum=<-50 -50 -50>, Maximum=<50 50 50>))
  ]
  particleUpdateScript [
   &ParticleState()
   &InheritVelocity(
    Inherited Velocity Amount Scale=&VectorFromFloat(Value=&FloatFromCurve(FloatCurve=DataInterface() {propertyValues {Curve {keys [<at 0 1 RCIM_Cubic>,<at 0.6585 0>]}}}, Scale Curve=-0.01))
    Speed Limit=true
    UseSpeedThreshold=false
   )
   &WindForce(
    Offset By Time:Vector3f=<0.5 0.25 0.2>
    Scale=0.673
    Wind Speed=&RandomRangeVector(Minimum=<-50 -50 -50>, Maximum=<50 50 50>)
    Wind Speed Scale=0.5
   )
   &AerodynamicDrag(
    Aerodynamic Drag:NiagaraFloat=&FloatFromCurve(FloatCurve=DataInterface() {propertyValues {Curve {keys [{interpMode=RCIM_Cubic, tangentMode=RCTM_User, time=0, value=0.2}, {interpMode=RCIM_Cubic, tangentMode=RCTM_User, time=1, value=1}]}}})
    Aerodynamic Rotational Drag=0.15
    Pivot Offset Maximum:Vector3f=<5 0.5 1>
    Rotation Strength:NiagaraFloat=&RandomRangeFloat(Minimum=0.2, Maximum=0.3)
    Use Linear Drag=true
    Use Rotational Drag=false
   )
   &CurlNoiseForce(Bias Noise=false, Noise Frequency=59.917, Noise Strength=500, Pan Noise=false, Pan Noise Field=<976.2582 0 500>)
   &Drag(Drag=0.4537, Rotational Drag=0.2513, Use Linear Drag=true, Use Rotational Drag=true)
   &SolveForcesAndVelocity(Clamp Velocity=false, Limit Acceleration=false)
   &ScaleSpriteSize(
    Uniform Curve Scale=0.9
    Uniform Curve Sprite Scale=DataInterface() {propertyValues {Curve {keys [
        {arriveTangent=14.5835, interpMode=RCIM_Cubic, leaveTangent=14.5835, tangentMode=RCTM_User, time=0, value=0}
        {arriveTangent=0.9132, interpMode=RCIM_Cubic, leaveTangent=0.9132, tangentMode=RCTM_User, time=0.1, value=1}
        {arriveTangent=-1.1111, interpMode=RCIM_Cubic, leaveTangent=-1.1111, time=0.5399, value=1}
        {arriveTangent=-3.3546, interpMode=RCIM_Cubic, leaveTangent=-3.3546, tangentMode=RCTM_User, time=1, value=0}
       ]}}}
   )
   &AlignSpriteToMeshOrientation(SetAlignment=true, SetFacing=true)
  ]
  rendererClasses [&NiagaraSpriteRendererProperties]
  renderers [{rendererClass=&NiagaraSpriteRendererProperties}]
 }
 Trail {
  emitterUpdateScript [
   &EmitterState()
   &FN_SpawnAnimTrail_v2()
  ]
  particleSpawnScript [
   /Game/Effects/Niagara/Modules/AnimTrail/FN_InitializeAnimTrail/FN_InitializeAnimTrailBoneIndecies(First Bone or Socket Index=1)
   &FN_InitializeAnimTrail(
    AnimTrailWidth=1.73
    Color=&LerpLinearColors(
     StartColor=<rgba 1.1 1.1 1.1 1>
     EndColor=<rgba 0.8646 0.8646 0.8646 1>
     LerpFactor=&FN_FrontEndFloatModifier(Gameplay=0, Front End=1)
    )
    Lifetime=0.09
    SkeletalMesh=$System.Skeletal Mesh
   ) {module=FN_InitializeAnimTrail001}
   &DynamicMaterialParameters(
    Index 0 Param 1=&RandomRangeFloat(Minimum=0, Maximum=1)
    Index 0 Param 2=3
    Index 0 Param 4=1.1
    Index 1 Param 2=0.2
    Index 1 Param 3=1
    Module.Index0Float0Write=true
    Module.Index0Float1Write=true
    Module.Index0Float2Write=true
    Module.Index0Float3Write=true
    Module.Index1Float0Write=true
    Module.Index1Float1Write=true
    Module.Index1Float2Write=true
    Module.Index1Float3Write=true
    Write Parameter Index 1=::NewEnumerator1
   )
   /SetVariables() {isSetParameters=true}
  ]
  particleUpdateScript [
   &FN_UpdateAnimTrail(SkeletalMesh=$System.Skeletal Mesh) {module=FN_UpdateAnimTrail001}
   &ScaleColor(
    Scale Alpha=&FloatFromCurve(
     FloatCurve=DataInterface() {propertyValues {Curve {keys [
         {arriveTangent=-2.0542, interpMode=RCIM_Cubic, leaveTangent=-2.0542, tangentMode=RCTM_User, time=0.0992, value=1}
         {arriveTangent=-1.0217, interpMode=RCIM_Cubic, leaveTangent=-1.0217, tangentMode=RCTM_User, time=0.8621, value=0}
        ]}}}
     Scale Curve=1.2
    )
    Scale RGB=&MakeVectorFromLinearColorRGB(Color=&FNTimeOfDayColor(
      Night Color=<rgba 0.8281 0.9029 1.5 1>
      Morning Color=<rgba 1.7 1.7 1.7 1>
      Day Color=<rgba 2 2 2 1>
      Evening Color=<rgba 1.6 1.3606 1.3417 1>
     ))
    ScaleA=true
    ScaleRGB=true
   )
   &CameraOffset(Camera Offset Amount=3)
   &ScaleRibbonWidth(Ribbon Width Scale=&FloatFromCurve(FloatCurve=DataInterface() {propertyValues {Curve {keys [<at 0.5 1 RCIM_Cubic>, {arriveTangent=-0.551, tangentMode=RCTM_User, time=1, value=0.4}]}}}))
   &DynamicMaterialParameters(
    Index 0 Param 4=&FloatFromCurve(FloatCurve=DataInterface() {propertyValues {Curve {keys [
         {interpMode=RCIM_Cubic, tangentMode=RCTM_User, time=0, value=1}
         {arriveTangent=-0.0857, interpMode=RCIM_Cubic, leaveTangent=-0.0857, tangentMode=RCTM_User, time=0.65, value=0}
        ]}}})
    Index 1 Param 1=&FloatFromCurve(FloatCurve=DataInterface() {propertyValues {Curve {keys [
         {interpMode=RCIM_Cubic, tangentMode=RCTM_User, time=0, value=0}
         {arriveTangent=-0.0857, interpMode=RCIM_Cubic, leaveTangent=-0.0857, tangentMode=RCTM_User, time=0.65, value=0.2}
        ]}}})
    Index 1 Param 4=&FloatFromCurve(FloatCurve=DataInterface() {propertyValues {Curve {keys [<at 0.4 1 RCIM_Cubic>,<at 0.9 0 RCIM_Cubic>]}}})
    Module.Index0Float0Write=false
    Module.Index0Float1Write=false
    Module.Index0Float2Write=false
    Module.Index0Float3Write=false
    Module.Index1Float0Write=false
    Module.Index1Float1Write=false
    Module.Index1Float2Write=false
    Module.Index1Float3Write=true
    Write Parameter Index 1=::NewEnumerator1
   ) {module=DynamicMaterialParameters001}
  ]
  rendererClasses [&NiagaraRibbonRendererProperties]
  renderers [{rendererClass=&NiagaraRibbonRendererProperties}]
 }
 TrailTip {
  emitterUpdateScript [
   &EmitterState()
   &FN_SpawnAnimTrail_v2()
  ]
  particleSpawnScript [
   &FN_InitializeAnimTrailBoneIndecies(Second Bone or Socket Index=2)
   &FN_InitializeAnimTrail(AnimTrailWidth=0.8, Color=<rgba 0.5256 0.5624 0.5833 1>, Lifetime=0.23, SkeletalMesh=$System.Skeletal Mesh) {module=FN_InitializeAnimTrail001}
  ]
  particleUpdateScript [
   &ScaleRibbonWidth(Ribbon Width Scale=&FloatFromCurve(FloatCurve=DataInterface() {propertyValues {Curve {keys [{arriveTangent=-1.2579, interpMode=RCIM_Cubic, time=0, value=0.7}, {interpMode=RCIM_Cubic, leaveTangent=-0.9664, time=1, value=0.22}]}}}))
   &ScaleColor(
    Scale Alpha=&Multiply_Float(
     A=&FloatFromCurve(
      FloatCurve=DataInterface() {propertyValues {Curve {keys [
          {interpMode=RCIM_Cubic, leaveTangent=-2.3646, tangentMode=RCTM_Break, time=0, value=1}
          <at 0.2103 0.13>
          {arriveTangent=-1.0846, interpMode=RCIM_Cubic, leaveTangent=-1.0846, tangentMode=RCTM_User, time=1, value=0}
         ]}}}
      Scale Curve=0.6
     )
     B=&FNTimeOfDayScalar(Night Scale=0.5, Morning Scale=0.6, Evening Scale=0.5)
    )
    Scale RGB=&MakeVectorFromLinearColorRGB(Color=&FNTimeOfDayColor(
      Night Color=<rgba 0.4463 0.62 1 1>
      Morning Color=<rgba 1 0.6899 0.5573 1>
      Day Color=<rgba 1 0.7878 0.6667 1>
      Evening Color=<rgba 1 0.6291 0.4464 1>
     ))
    ScaleA=true
    ScaleRGB=true
   ) {module=ScaleColor002}
   &FN_UpdateAnimTrail(SkeletalMesh=$System.Skeletal Mesh) {module=FN_UpdateAnimTrail001}
  ]
  rendererClasses [&NiagaraRibbonRendererProperties]
  renderers [{rendererClass=&NiagaraRibbonRendererProperties}]
 }
}
summary {
 systemName=NS_Pickaxe_GarlicWhisk_Trail
 emitters [
  {emitterName=Trail, rendererClasses [&NiagaraRibbonRendererProperties]}
  {emitterName=TrailTip, rendererClasses [&NiagaraRibbonRendererProperties]}
  {emitterName=Petals, rendererClasses [&NiagaraSpriteRendererProperties]}
 ]
 userVariables [
  User.bIsFrontEnd {type=NiagaraBool}
  User.bIsFrontEndPreview {type=NiagaraBool}
  User.RibbonIDOffset {type=NiagaraInt32}
 ]
}
